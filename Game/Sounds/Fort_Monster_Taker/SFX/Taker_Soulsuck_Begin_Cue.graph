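# Sound cue for the start of the Taker monster's soul-suck, judging by its name.
# A single impact wave player feeds a Modulator node left at its usual settings, which is the cue's root.
# It borrows the small melee weapon attenuation and melee weapon sound class, and overrides concurrency so new instances are prevented once the limit is reached.
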SoundCue /Game/Sounds/Fort_Monster_Taker/SFX/Taker_Soulsuck_Begin_Cue
root=Modulator_5
nodes [
 {id=Modulator_5, children [WavePlayer_27]}
 {id=WavePlayer_27, properties {soundWaveAssetPtr=&/taker_sfx_impact}}
]
properties {
 attenuationSettings=&Fort_Weapon_Melee_Sml_Att
 bOverrideConcurrency=true
 maxDistance=2800
 soundClassObject=&Fort_WeaponMelee_PSM
 concurrencyOverrides {resolutionRule=PreventNew}
}
